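# Translucent UI material for GravWell HUD pip decoration, drawing detail shapes from the green channel of an SDF texture.
# A SmoothStep cuts the main shape and blends black to the Color parameter.
# A second SmoothStep with its own range forms a shadow mask, blended with the shape by ShadowOpacity and scaled by Opacity.
# Two MF_UI_AntiAliasing calls share an input-less OneMinus and feed no other node.
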Material /GravWellGameplay/HUD/Materials/M_UI_GravWell_PipDecor
graph {
 properties {bAutomaticallySetUsageInEditor=false, blendMode=BLEND_Translucent, materialDomain=MD_UI, usedShadingModels=Default Lit}
 nodes [
  n1 = TextureSample(UVs=TextureCoordinate()) {texture=/GravWellGameplay/HUD/Assets/T_UI_GravWell_Details_SDF}
  n2 = SmoothStep(
   Min=ScalarParameter() {defaultValue=0.3, parameterName=SmoothStepMin}
   Max=ScalarParameter() {defaultValue=0.5, parameterName=SmoothStepMax}
   Value=@n1.G
  )
  LinearInterpolate(B=VectorParameter() {parameterName=Color, defaultValue=<rgba 1 1 1 0>}.RGB, Alpha=@n2)
  n3 = OneMinus()
  MaterialFunctionCall(DistanceField=@n3, Threshold=Constant() {r=0.71}) {materialFunction=&MF_UI_AntiAliasing}
  MaterialFunctionCall(DistanceField=@n3, Threshold=Constant() {r=0.99}) {materialFunction=&MF_UI_AntiAliasing}
  n4 = ScalarParameter() {parameterName=SmoothStepMinShadow}
  n5 = ScalarParameter() {defaultValue=0.1, parameterName=SmoothStepMaxShadow}
  Multiply(
   A=LinearInterpolate(A=@n2, B=SmoothStep(Min=@n4, Max=@n5, Value=@n1.G), Alpha=ScalarParameter() {defaultValue=0.3, parameterName=ShadowOpacity})
   B=ScalarParameter() {defaultValue=0.5, parameterName=Opacity}
  )
 ]
}
