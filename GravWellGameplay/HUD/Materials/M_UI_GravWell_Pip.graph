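# Translucent UI material for a GravWell HUD pip, drawing a star that fills vertically with a Progress parameter.
# The star comes from mirrored UVs feeding an SDF circle outline, anti-aliased at two thresholds for shape and shadow backing.
# A step on vertical UV masks the fill, colored by a vertical two-color gradient or a flat color.
# Named reroutes carry these masks, plus a box-distance border, into color and opacity blends with adjustable fill and shadow opacity.
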Material /GravWellGameplay/HUD/Materials/M_UI_GravWell_Pip
graph {
 properties {bAutomaticallySetUsageInEditor=false, blendMode=BLEND_Translucent, materialDomain=MD_UI, usedShadingModels=Default Lit}
 nodes [
  n1 = NamedRerouteUsage()
  n2 = Multiply(A=NamedRerouteUsage(), B=@n1)
  LinearInterpolate(A=ScalarParameter() {parameterName=ShadowColor}, B=NamedRerouteUsage(), Alpha=@n2)
  n3 = Multiply(A=LinearInterpolate(A=TextureCoordinate(), B=TextureCoordinate() {coordinateIndex=1}, Alpha=ScalarParameter() {parameterName=CubeMap_Switch})) {constB=2}
  n4 = SquareRoot(None=Abs(None=Subtract(A=@n3)))
  n5 = Divide(
   A=MaterialFunctionCall(Float2=@n4) {materialFunction=&BreakOutFloat2Components}.G
   B=MaterialFunctionCall(Float2=OneMinus(None=@n4)) {materialFunction=&BreakOutFloat2Components}.R
  )
  MaterialFunctionCall(DistanceField=OneMinus(None=Subtract(B=@n5)), Threshold=Constant() {r=1}) {materialFunction=&MF_UI_AntiAliasing}
  n6 = VectorParameter() {parameterName=GradientColor01, defaultValue=<rgba 0.314 0.1559 0.9047 1>}
  n7 = VectorParameter() {parameterName=GradientColor02, defaultValue=<rgba 0.8796 0.15 1 1>}
  n8 = TextureCoordinate()
  n9 = Power(Base=Add(A=ComponentMask(None=@n8) {g=true}) {constB=0.3}) {constExponent=5}
  NamedRerouteDeclaration(None=LinearInterpolate(
    A=VectorParameter() {parameterName=Color03, defaultValue=<rgba 0.0437 0.956 1 1>}.RGB
    B=LinearInterpolate(A=@n6.RGB, B=@n7.RGB, Alpha=@n9)
    Alpha=ScalarParameter() {defaultValue=1, parameterName=Gradient}
   )) {name=FillColor, nodeColor=<rgba 0 0 0 1>}
  n10 = MaterialFunctionCall!(UV=TextureCoordinate(), Swizzle UV) {materialFunction=/FortUILibrary/MaterialFunctions/Basics/MF_UI_Symmetry_2x2_New}
  n11 = ScalarParameter() {defaultValue=0.49, group=Star, parameterName=CircleOffset, sortPriority=0}
  n12 = MaterialFunctionCall(
   UVs=MaterialFunctionCall(UVs=Add(A=@n10.Symmetry, B=@n11), Texture Scale=Constant() {r=0.5}) {materialFunction=&ScaleUVsByCenter}.UVs
   Size=ScalarParameter() {defaultValue=4.2, group=Star, parameterName=Size, sortPriority=1}
   OutlineThickness=ScalarParameter() {defaultValue=0.32, group=Star, parameterName=OutlineTickness, sortPriority=2}
   OutlineAlignment=Constant() {r=1}
  ) {materialFunction=&MF_UI_Circle}
  NamedRerouteDeclaration(None=MaterialFunctionCall(DistanceField=@n12.SDFOutline, Threshold=ScalarParameter()) {materialFunction=&MF_UI_AntiAliasing}.Inner) {name=StarShape, nodeColor=<rgba 0 0 0 1>}
  NamedRerouteDeclaration(None=MaterialFunctionCall(DistanceField=@n12.SDFOutline, Threshold=ScalarParameter() {defaultValue=0.15, group=Star, parameterName=ShadowThickness, sortPriority=3}) {materialFunction=&MF_UI_AntiAliasing}.Inner) {name=BackgroundOpacity, nodeColor=<rgba 0 0 0 1>}
  n13 = ComponentMask(None=TextureCoordinate()) {g=true}
  n14 = ScalarParameter() {defaultValue=1, parameterName=Progress}
  NamedRerouteDeclaration(None=Step(Y=OneMinus(None=@n13), X=LinearInterpolate(Alpha=@n14) {constA=0.18, constB=0.82})) {name=FillProgress, nodeColor=<rgba 0 0 0 1>}
  n15 = TextureCoordinate()
  n16 = Subtract(A=Abs(None=ConstantBiasScale(None=@n15) {bias=-0.5, scale=2}), B=Constant2Vector() {g=0.7, r=0.7}) {constB=0.5}
  NamedRerouteDeclaration(None=Length(None=Max(A=@n16) {constB=0}) {is=class MaterialExpressionLength}) {name=Border, nodeColor=<rgba 0 0 0 1>}
  n17 = LinearInterpolate(A=@n1, B=@n2, Alpha=Subtract(B=ScalarParameter() {defaultValue=1.2, parameterName=FillOpacity}) {constA=2})
  n18 = NamedRerouteUsage()
  n19 = ScalarParameter() {defaultValue=0.3, parameterName=ShadowOpacity}
  Saturate(None=Subtract(A=LinearInterpolate(A=@n17, B=@n18, Alpha=@n19), B=NamedRerouteUsage()))
 ]
}
