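# Charge pip widget for the GravWell weapon reticle, extending the shared recharge-stack pip base.
# Construct registers Image_Pip with the base and applies the purple-pink default material state.
# InitFillState sets fill percent and plays either Anim_Fill or Anim_Use, each stopping the other.
# These animations drive material gradients and opacity between white, dimmed empty and colored full looks, while OnFill and OnUse set fill directly, with OnUse also whitening the gradient.
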WidgetBlueprint /GravWellGameplay/HUD/WBP_GravWell_Reticle_Pip
animations [
 Anim_Fill {
  binding_count=1
  length_seconds=0.5
  object=&self:Anim_Fill
  sequence {bindings [Image_Pip {
     id=#1
     possessed_class=Image
     tracks [
      {
       type=MovieSceneWidgetMaterialTrack
       sections [{
         type=MovieSceneParameterSection
         end_frame=10
         start_frame=0
         channels [
          Gradient {type=MovieSceneScriptingFloatChannel, keys [[0,1],[3,1],[7,0],[10,1]]}
          ShadowColor {type=MovieSceneScriptingFloatChannel, keys [[0,0],[3,0],[7,1],[10,0]]}
          GradientColor01R {type=MovieSceneScriptingFloatChannel, keys [[0,1],[7,0.314],[10,0.314]]}
          GradientColor01G {type=MovieSceneScriptingFloatChannel, keys [[0,1],[7,0.1559],[10,0.1559]]}
          GradientColor01B {type=MovieSceneScriptingFloatChannel, keys [[0,1],[7,0.9047],[10,0.9047]]}
          GradientColor01A {type=MovieSceneScriptingFloatChannel, const [1,2]}
          GradientColor02R {type=MovieSceneScriptingFloatChannel, keys [[0,1],[7,0.8796],[10,0.8796]]}
          GradientColor02G {type=MovieSceneScriptingFloatChannel, keys [[0,1],[7,0.15],[10,0.15]]}
          GradientColor02B {type=MovieSceneScriptingFloatChannel, const [1,3]}
          GradientColor02A {type=MovieSceneScriptingFloatChannel, const [1,2]}
         ]
        }]
      }
      {
       type=MovieSceneFloatTrack
       display_name=Render Opacity
       sections [{
         type=MovieSceneFloatSection
         end_frame=10
         start_frame=0
         channels [None {type=MovieSceneScriptingFloatChannel, keys [[0,0.5],[3,0.5],[7,1],[10,1]]}]
        }]
      }
     ]
    }]}
 }
 Anim_Use {
  binding_count=1
  length_seconds=0.5
  object=&self:Anim_Use
  sequence {bindings [Image_Pip {
     id=#2
     possessed_class=Image
     tracks [
      {
       type=MovieSceneWidgetMaterialTrack
       sections [{
         type=MovieSceneParameterSection
         end_frame=10
         start_frame=0
         channels [
          Size {type=MovieSceneScriptingFloatChannel, keys [[0,4.2],[3,4.5217],[5,4.2],[10,4.2]]}
          GradientColor01R {type=MovieSceneScriptingFloatChannel, keys [[0,0.314],[3,0.314],[5,1],[10,1]]}
          GradientColor01G {type=MovieSceneScriptingFloatChannel, keys [[0,0.1559],[3,0.1559],[5,1],[10,1]]}
          GradientColor01B {type=MovieSceneScriptingFloatChannel, keys [[0,0.9047],[3,0.9047],[5,1],[10,1]]}
          GradientColor01A {type=MovieSceneScriptingFloatChannel, const [1,2]}
          GradientColor02R {type=MovieSceneScriptingFloatChannel, keys [[0,0.8796],[3,0.8796],[5,1],[10,1]]}
          GradientColor02G {type=MovieSceneScriptingFloatChannel, keys [[0,0.15],[3,0.15],[5,1],[10,1]]}
          GradientColor02B {type=MovieSceneScriptingFloatChannel, const [1,4]}
          GradientColor02A {type=MovieSceneScriptingFloatChannel, const [1,2]}
         ]
        }]
      }
      {
       type=MovieSceneFloatTrack
       display_name=Render Opacity
       sections [{
         type=MovieSceneFloatSection
         end_frame=10
         start_frame=0
         channels [None {type=MovieSceneScriptingFloatChannel, keys [[0,1],[3,1],[5,0.5],[10,0.5]]}]
        }]
      }
     ]
    }]}
 }
]
graphs [
 EventGraph {nodes [
   Event Construct()
   > SetImagePip!(
    execute
    self:Fort Weapon Reticle Recharge Stack Base Pip Base Object Reference
    Image:Image Object Reference=Get Image_Pip() -> (Image_Pip:Image Object Reference)
   ) -> (then)
   > SetDefaultState()
   Event PreConstruct()
  ]}
 InitFillState {nodes [
   n1 = InitFillState!() -> (then, IsFilled:bool) {is=title K2Node_FunctionEntry}
   > Parent: InitFillState!(execute, IsFilled:bool=@n1) -> (then) {is=title K2Node_CallParentFunction}
   n2 = Get IsFilled() -> (IsFilled:bool)
   > SetFillPercent!(
    execute
    self:Fort Weapon Reticle Recharge Stack Base Pip Base Object Reference
    FillPercent:float=SelectFloat(A=1, B=0, bPickA=@n2)
    PipPercentMatParamName:Name=Get ImagePipFillParam() -> (ImagePipFillParam:Name=None)
   ) -> (then)
   > n3 = Branch(Condition=@n2)
   > PlayFillAnimation(self:WBP Reticle Recharge Stack Base Pip Object Reference)
   PlayUseAnimation(execute=@n3.else, self:WBP Reticle Recharge Stack Base Pip Object Reference)
  ]}
 OnFill {nodes [
   OnFill() {is=title K2Node_FunctionEntry}
   > Parent: OnFill() {is=title K2Node_CallParentFunction}
   > SetFillPercent(FillPercent=1, PipPercentMatParamName=Get ImagePipFillParam() -> (ImagePipFillParam=None))
   SetDefaultState()
  ]}
 OnUse {nodes [
   OnUse() {is=title K2Node_FunctionEntry}
   > Parent: OnUse() {is=title K2Node_CallParentFunction}
   > SetFillPercent(FillPercent=0, PipPercentMatParamName=Get ImagePipFillParam() -> (ImagePipFillParam=None))
   > SetOnUseState()
  ]}
 PlayFillAnimation {nodes [
   PlayFillAnimation() {is=title K2Node_FunctionEntry}
   > Parent: PlayFillAnimation() {is=title K2Node_CallParentFunction}
   > StopAnimation(InAnimation=Get Anim_Use() -> (Anim_Use:Widget Animation Object Reference))
   > PlayAnimation(InAnimation=Get Anim_Fill() -> (Anim_Fill:Widget Animation Object Reference))
  ]}
 PlayUseAnimation {nodes [
   PlayUseAnimation() {is=title K2Node_FunctionEntry}
   > Parent: PlayUseAnimation() {is=title K2Node_CallParentFunction}
   > StopAnimation(InAnimation=Get Anim_Fill())
   > PlayAnimation(InAnimation=Get Anim_Use())
  ]}
 SetDefaultState {nodes [
   SetDefaultState() {is=title K2Node_FunctionEntry}
   n1 = Get Image_Pip()
   > n2 = GetDynamicMaterial(self=@n1)
   > SetScalarParameterValue(self=@n2, ParameterName=Size, Value=4.2)
   > n3 = GetDynamicMaterial(self=@n1)
   > SetVectorParameterValue(self=@n3, ParameterName=GradientColor01, Value=(R=0.314,G=0.1559,B=0.9047,A=1))
   > SetVectorParameterValue(self=@n3, ParameterName=GradientColor02, Value=(R=0.8796,G=0.15,B=1,A=1))
   SetOpacity!(execute, self:Image Object Reference, InOpacity:float=1) -> (then)
  ]}
 SetOnUseState {nodes [
   SetOnUseState() {is=title K2Node_FunctionEntry}
   > n1 = GetDynamicMaterial(self=Get Image_Pip())
   > SetVectorParameterValue(self=@n1, ParameterName=GradientColor01, Value=(R=1,G=1,B=1,A=1))
   > SetVectorParameterValue(self=@n1, ParameterName=GradientColor02, Value=(R=1,G=1,B=1,A=1))
   SetOpacity(InOpacity=0.5)
  ]}
]
variables [ImagePipFillParam {type=string, default=Progress}]
widgets {
 info {
  inheritedWidgetCount=0
  parentClass=/Game/Athena/HUD/reticles/ReticleExtensionWidgets/ChargeStack/WBP_Reticle_RechargeStackBase_Pip.WBP_Reticle_RechargeStackBase_Pip_C
  rootWidgetClass=&Overlay
  widgetCount=3
 }
 widgets [
  Overlay_150 {widgetClassPath=&Overlay, widget=Overlay_150}
  Image_Pip {
   widgetClassPath=&Image
   bIsVariable=true
   parent=Overlay_150
   widget=Image_Pip
   properties {brush {
     bIsDynamicallyLoaded=false
     drawAs=Image
     imageType=FullColor
     mirroring=NoMirror
     resourceName=None
     resourceObject=/GravWellGameplay/HUD/Materials/M_UI_GravWell_Pip
     tiling=NoTile
     imageSize=<30 30>
     margin=<ltrb 0 0 0 0>
     outlineSettings {
      bUseBrushTransparency=false
      roundingType=HalfHeightRadius
      width=0
      color {colorUseRule=UseColor_Specified, specifiedColor=<rgba 0 0 0 0>}
      cornerRadii=<0 0 0 0>
     }
     tintColor {colorUseRule=UseColor_Specified, specifiedColor=<rgba 1 1 1 1>}
     uVRegion {bIsValid=false, max=<0 0>, min=<0 0>}
    }}
   slot {class=OverlaySlot}
  }
  Spacer_36 {
   widgetClassPath=&Spacer
   parent=Overlay_150
   widget=Spacer_36
   properties {size=<10 10>}
   slot {class=OverlaySlot, properties {horizontalAlignment=HAlign_Fill, verticalAlignment=VAlign_Fill}}
  }
 ]
}
