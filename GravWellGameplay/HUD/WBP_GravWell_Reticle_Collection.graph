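# Grav Well weapon reticle widget that shows recharge stacks as a row of pips, built on the shared recharge-stack pip collection base.
# A DynamicEntryBox of WBP_GravWell_Reticle_Pip entries sits between two decor images, the right one mirrored by negative render scale.
# On Construct it passes that box to the base, then plays Anim_Equip, which fades the row in while both decor images slide inward.
# Its GenerateStacks override only forwards to the parent.
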WidgetBlueprint /GravWellGameplay/HUD/WBP_GravWell_Reticle_Collection
animations [Anim_Equip {
  binding_count=3
  length_seconds=0.25
  object=&self:Anim_Equip
  sequence {bindings [
    Image_DetailL {
     id=#1
     possessed_class=Image
     tracks [{
       type=MovieScene2DTransformTrack
       sections [{
         type=MovieScene2DTransformSection
         end_frame=5
         start_frame=0
         channels [
          Translation.X {type=MovieSceneScriptingFloatChannel, keys [[0,-6],[5,0]]}
          Translation.Y {type=MovieSceneScriptingFloatChannel, const [0,1]}
         ]
        }]
      }]
    }
    Image_DetailR {
     id=#2
     possessed_class=Image
     tracks [{
       type=MovieScene2DTransformTrack
       sections [{
         type=MovieScene2DTransformSection
         end_frame=5
         start_frame=0
         channels [
          Translation.X {type=MovieSceneScriptingFloatChannel, keys [[0,6],[5,0]]}
          Translation.Y {type=MovieSceneScriptingFloatChannel, const [0,1]}
         ]
        }]
      }]
    }
    HorizontalBox_0 {
     id=#3
     possessed_class=HorizontalBox
     tracks [{
       type=MovieSceneFloatTrack
       display_name=Render Opacity
       sections [{type=MovieSceneFloatSection, end_frame=5, start_frame=0, channels [None {type=MovieSceneScriptingFloatChannel, keys [[0,0],[5,1]]}]}]
      }]
    }
   ]}
 }]
graphs [
 EventGraph {nodes [
   OnEquip() {is=title K2Node_CustomEvent}
   > PlayAnimationForward(InAnimation=Get Anim_Equip() -> (Anim_Equip:Widget Animation Object Reference))
   Event Construct()
   > SetDynamicEntryBox!(
    execute
    self:Fort Weapon Reticle Recharge Stack Base Pip Collection Base Object Reference
    DynamicEntryBox:Dynamic Entry Box Object Reference=Get DynamicEntryBox_Pips() -> (DynamicEntryBox_Pips:Dynamic Entry Box Object Reference)
   ) -> (then)
   > OnEquip() {is=title K2Node_CallFunction}
  ]}
 GenerateStacks {nodes [
   n1 = GenerateStacks!() -> (then, TotalStacks:int=0, CurrentStacks:int=0) {is=title K2Node_FunctionEntry}
   > Parent: GenerateStacks!(execute, TotalStacks:int=@n1, CurrentStacks:int=@n1.CurrentStacks) -> (then) {is=title K2Node_CallParentFunction}
  ]}
]
variables []
widgets {
 info {
  inheritedWidgetCount=0
  parentClass=/Game/Athena/HUD/reticles/ReticleExtensionWidgets/ChargeStack/WBP_Reticle_RechargeStackBase_PipCollection.WBP_Reticle_RechargeStackBase_PipCollection_C
  rootWidgetClass=&SizeBox
  widgetCount=5
 }
 widgets [
  SizeBox_20 {widgetClassPath=&SizeBox, widget=SizeBox_20, properties {heightOverride=64, widthOverride=82}}
  HorizontalBox_0 {
   widgetClassPath=&HorizontalBox
   parent=SizeBox_20
   widget=HorizontalBox_0
   slot {class=SizeBoxSlot, properties {horizontalAlignment=HAlign_Center, verticalAlignment=VAlign_Center}}
  }
  Image_DetailL {
   widgetClassPath=&Image
   parent=HorizontalBox_0
   widget=Image_DetailL
   properties {brush {
     bIsDynamicallyLoaded=false
     drawAs=Image
     imageType=FullColor
     mirroring=NoMirror
     resourceName=None
     resourceObject=/GravWellGameplay/HUD/Materials/M_UI_GravWell_PipDecor
     tiling=NoTile
     imageSize=<25 25>
     margin=<ltrb 0 0 0 0>
     outlineSettings {
      bUseBrushTransparency=false
      roundingType=HalfHeightRadius
      width=0
      color {colorUseRule=UseColor_Specified, specifiedColor=<rgba 0 0 0 0>}
      cornerRadii=<0 0 0 0>
     }
     tintColor {colorUseRule=UseColor_Specified, specifiedColor=<rgba 1 1 1 1>}
     uVRegion {bIsValid=false, max=<0 0>, min=<0 0>}
    }}
   slot {class=HorizontalBoxSlot, properties {horizontalAlignment=HAlign_Center, verticalAlignment=VAlign_Center, padding=<ltrb 0 0.5 0 0>}}
  }
  DynamicEntryBox_Pips {
   widgetClassPath=&DynamicEntryBox
   bIsVariable=true
   parent=HorizontalBox_0
   widget=DynamicEntryBox_Pips
   properties {
    entryWidgetClass=&/WBP_GravWell_Reticle_Pip.WBP_GravWell_Reticle_Pip_C
    entrySizeRule {sizeRule=Automatic, value=0.5}
    entrySpacing=<-1 0>
    radialBoxSettings {angleBetweenItems=45, bClockwiseOrder=false, bDistributeItemsEvenly=true, marginSize=0, sectorCentralAngle=115, startingAngle=256}
   }
   slot {class=HorizontalBoxSlot}
  }
  Image_DetailR {
   widgetClassPath=&Image
   parent=HorizontalBox_0
   widget=Image_DetailR
   properties {
    brush {
     bIsDynamicallyLoaded=false
     drawAs=Image
     imageType=FullColor
     mirroring=NoMirror
     resourceName=None
     resourceObject=&M_UI_GravWell_PipDecor
     tiling=NoTile
     imageSize=<25 25>
     margin=<ltrb 0 0 0 0>
     outlineSettings {
      bUseBrushTransparency=false
      roundingType=HalfHeightRadius
      width=0
      color {colorUseRule=UseColor_Specified, specifiedColor=<rgba 0 0 0 0>}
      cornerRadii=<0 0 0 0>
     }
     tintColor {colorUseRule=UseColor_Specified, specifiedColor=<rgba 1 1 1 1>}
     uVRegion {bIsValid=false, max=<0 0>, min=<0 0>}
    }
    renderTransform {angle=0, scale=<-1 1>, shear=<0 0>, translation=<0 0>}
   }
   slot {class=HorizontalBoxSlot, properties {horizontalAlignment=HAlign_Center, verticalAlignment=VAlign_Center, padding=<ltrb 0 0.5 0 0>}}
  }
 ]
}
